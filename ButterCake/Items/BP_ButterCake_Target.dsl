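; BP_ButterCake_Target is a ButterCake item target actor, likely a landing or aim marker, that lasts until its owner hits something.
; On BeginPlay it binds SelfDestroy to a valid owner's OnActorHit and starts a 30-second fail-safe timer for AutoDestroy.
; Both handlers unbind all of the owner's hit events, not just their own, then set Hidden In Game false and destroy the target.
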(event EventBeginPlay
  (bind _returnvalue (Actor|GetOwner))
  (|CallParentFunction)
  ;[1 1 1] Fail safe, if we don't recieve event from Owner
  (Utilities|Time|SetTimerbyEvent (delegate Custom|AutoDestroy) 3e+01)
  (Utilities|IsValid _returnvalue
    (:Is%20Valid
      (Collision|BindEventtoOnActorHit _returnvalue (delegate Custom|SelfDestroy)))))

(event Collision|EventActorBeginOverlap
  (params (OtherActor :type Actor%20Object%20Reference)))

(event EventTick
  (params (DeltaSeconds :type Float%20%28single-precision%29)))

(event Custom|AutoDestroy
  (Collision|UnbindallEventsfromOnActorHit (Actor|GetOwner))
  (Rendering|SetActorHiddenInGame false)
  (Actor|DestroyActor))

(event Custom|SelfDestroy
  (params (SelfActor :type Actor%20Object%20Reference) (OtherActor :type Actor%20Object%20Reference) (NormalImpulse :type Vector) (Hit :type Hit%20Result%20Structure%20%28by%20ref%29))
  (Collision|UnbindallEventsfromOnActorHit (Actor|GetOwner))
  (Rendering|SetActorHiddenInGame false)
  (Actor|DestroyActor))

(fn ConstructionScript)
